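# Unlit translucent Niagara mesh-particle material for the BoomShot pickaxe's smoke puff, shaded with comic-style Ben-Day dots.
# A parameter sun vector and vertex normal give fake lighting, with Fresnel rim on the dark side.
# That lighting sizes aspect-corrected screen-space dots, growing as light falls, and cuts to flat shadow color below a threshold.
# Particle color tints the result, while opacity combines particle alpha, noise dissolve driven by dynamic Erosion, and near-camera fade.
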Material /BRCosmetics/Weapons/FORT_Melee/Pickaxe_BoomShot/FX/M_Pickaxe_BoomShot_BenDaySmokePuff
graph {
 properties {
  bAutomaticallySetUsageInEditor=false
  bUsedWithNiagaraMeshParticles=true
  blendMode=BLEND_Translucent
  shadingModel=MSM_Unlit
  usedShadingModels=Unlit
 }
 nodes [
  n1 = VectorParameter() {group=Shading, parameterName=Shadow Color, sortPriority=2, defaultValue=<rgba 0.0885 0 0.0653 1>}
  n2 = MaterialFunctionCall() {materialFunction=&ScreenResolution}
  n3 = ScalarParameter() {defaultValue=150, group=Shading, parameterName=Dots Tiling, sliderMax=100}
  n4 = Multiply(A=Divide(A=ComponentMask(None=@n2.Visible Resolution) {r=true}, B=ComponentMask(None=@n2.Visible Resolution) {g=true}), B=@n3)
  n5 = Frac(None=Multiply(A=ScreenPosition().ViewportUV, B=AppendVector(A=@n4, B=@n3)) {constB=50})
  n6 = Constant() {r=0.5}
  n7 = ScalarParameter() {defaultValue=0.4, group=Shading, parameterName=Dot Size, sliderMax=0.5}
  n8 = VertexNormalWS()
  n9 = DotProduct(A=Normalize(VectorInput=VectorParameter() {group=Shading, parameterName=SunDirectionVector, defaultValue=<rgba 0 0 1 1>}.RGB), B=@n8)
  n10 = Saturate(None=@n9)
  n11 = LinearInterpolate(A=Constant(), B=Fresnel(Normal=@n8) {baseReflectFraction=0.1, exponent=3}, Alpha=Saturate(None=Multiply(A=@n9) {constB=-1}))
  n12 = Power(
   Base=SmoothStep(
    Min=ScalarParameter() {group=Shading, parameterName=Smooth Step Min, sliderMax=1, sliderMin=-1, sortPriority=3}
    Max=ScalarParameter() {defaultValue=1, group=Shading, parameterName=Smooth Step Max, sliderMax=1, sortPriority=4}
    Value=Add(A=@n10, B=@n11)
   )
   Exp=ScalarParameter() {defaultValue=1, group=Shading, parameterName=Color Power, sliderMax=5, sortPriority=5}
  )
  n13 = OneMinus(None=@n12)
  n14 = LinearInterpolate(
   A=VectorParameter() {group=Shading, parameterName=Base Color, sortPriority=0, defaultValue=<rgba 1 0 0.7385 1>}.RGB
   B=VectorParameter() {group=Shading, parameterName=Dot Color, sortPriority=1, defaultValue=<rgba 0.3698 0 0.2705 1>}.RGB
   Alpha=If(A=Distance(A=@n5, B=@n6), B=Multiply(A=@n7, B=@n13), "A > B"=Constant(), "A < B"=Constant() {r=1}) {constB=0.2}
  )
  n15 = If(
   A=@n12
   B=ScalarParameter() {defaultValue=0.001, group=Shading, parameterName=Shadow Threshold, sliderMax=0.1}
   "A > B"=Constant() {r=1}
   "A < B"=Constant()
  )
  n16 = ParticleColor()
  MaterialFunctionCall(Emissive=Multiply(A=LinearInterpolate(A=@n1.RGB, B=@n14, Alpha=@n15), B=@n16.RGB)) {materialFunction=&MF_FX_EmissiveAdjustments}
  n17 = OneMinus(None=TextureSample(UVs=TextureCoordinate() {uTiling=2, vTiling=2}) {samplerType=SAMPLERTYPE_Grayscale, texture=&LowResBlurredNoise}.RGB)
  n18 = DynamicParameter() {defaultValue=<rgba 0 0 1 1>, paramNames [Erosion, Random X, Param3, Param4]}
  n19 = ScalarParameter() {defaultValue=5, group=Shading, parameterName=Erosion Contrast, sliderMax=20}
  Multiply(
   A=Multiply(
    A=@n16.A
    B=MaterialFunctionCall!(Ramp=@n17, Dissolve=@n18.Erosion, EdgeSharpness=@n19) {materialFunction=/Game/Effects/Fort_Effects/Materials/_Functions/MF_Dissolve}.ClampedResult
   )
   B=MaterialFunctionCall(Fade Length=ScalarParameter() {defaultValue=256, parameterName=Near Camera Fade Distance, sliderMax=1024}) {materialFunction=&CameraDepthFade}.Result
  )
 ]
}
