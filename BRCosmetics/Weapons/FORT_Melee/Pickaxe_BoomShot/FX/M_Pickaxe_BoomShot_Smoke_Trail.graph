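# Translucent unlit Niagara ribbon material for the BoomShot pickaxe's stylized smoke trail.
# Normal-map lighting against a sun vector parameter splits shadow from base color, with aspect-corrected screen-space halftone dots sized by shading and a leading color along U.
# Panned, stretched UVs feed the textures; a warped sphere mask and erosion dissolve shape opacity, followed by depth and optional front, view-angle and near-camera fades.
# Dynamic parameters drive erosion, U offset and color swipe.
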Material /BRCosmetics/Weapons/FORT_Melee/Pickaxe_BoomShot/FX/M_Pickaxe_BoomShot_Smoke_Trail
graph {
 properties {
  bAutomaticallySetUsageInEditor=false
  bUsedWithNiagaraRibbons=true
  blendMode=BLEND_Translucent
  shadingModel=MSM_Unlit
  usedShadingModels=Unlit
 }
 nodes [
  n1 = VectorParameter() {group=Shading, parameterName=Shadow Color, sortPriority=2, defaultValue=<rgba 0.0885 0 0.0653 1>}
  n2 = DynamicParameter() {defaultValue=<rgba 0 0 0.1 1>, paramNames [Erosion, Random X, Leading Color Swipe, Param4]}
  n3 = Add(A=ScalarParameter() {parameterName=Leading Color Start, sliderMax=1}, B=@n2.Leading Color Swipe)
  n4 = Add(A=ScalarParameter() {defaultValue=0.3, parameterName=Leading Color End, sliderMax=1}, B=@n2.Leading Color Swipe)
  n5 = ComponentMask(None=TextureCoordinate()) {r=true}
  n6 = MaterialFunctionCall() {materialFunction=&ScreenResolution}
  n7 = ScalarParameter() {defaultValue=150, group=Shading, parameterName=Dots Tiling, sliderMax=100}
  n8 = Multiply(A=Divide(A=ComponentMask(None=@n6.Visible Resolution) {r=true}, B=ComponentMask(None=@n6.Visible Resolution) {g=true}), B=@n7)
  n9 = Frac(None=Multiply(A=ScreenPosition().ViewportUV, B=AppendVector(A=@n8, B=@n7)) {constB=50})
  n10 = Constant() {r=0.5}
  n11 = ScalarParameter() {defaultValue=0.25, group=Shading, parameterName=Dot Size, sliderMax=0.5}
  n12 = TextureSampleParameter2D!(UVs=NamedRerouteUsage(), Bias, Apply View MipBias) {
   group=Texture
   mipValueMode=TMVM_MipBias
   parameterName=Texture Normal
   samplerType=SAMPLERTYPE_Normal
   texture=/Game/Effects/Fort_Effects/Gameplay/SmokeDustExplosions/Textures/T_RibbonSmokeMasked_01
  }
  n13 = Normalize(VectorInput=Transform(None=@n12.RGB))
  n14 = DotProduct(A=Normalize(VectorInput=VectorParameter() {group=Shading, parameterName=SunDirectionVector, defaultValue=<rgba 0 0 1 1>}.RGB), B=@n13)
  n15 = Saturate(None=@n14)
  n16 = LinearInterpolate(A=Constant(), B=Fresnel(Normal=@n13) {baseReflectFraction=0.05}, Alpha=Saturate(None=Multiply(A=@n14) {constB=-1}))
  n17 = Power(
   Base=SmoothStep(
    Min=ScalarParameter() {defaultValue=0.2, group=Shading, parameterName=Smooth Step Min, sliderMax=1, sliderMin=-1, sortPriority=3}
    Max=ScalarParameter() {defaultValue=1, group=Shading, parameterName=Smooth Step Max, sliderMax=1, sortPriority=4}
    Value=Add(A=@n15, B=@n16)
   )
   Exp=ScalarParameter() {defaultValue=3, group=Shading, parameterName=Color Power, sliderMax=5, sortPriority=5}
  )
  n18 = OneMinus(None=@n17)
  n19 = LinearInterpolate(
   A=LinearInterpolate(
    A=VectorParameter() {parameterName=Leading Color, defaultValue=<rgba 1 0.3333 0 1>}.RGB
    B=VectorParameter() {group=Shading, parameterName=Base Color, sortPriority=0, defaultValue=<rgba 1 0 0.7385 1>}.RGB
    Alpha=SmoothStep(Min=@n3, Max=@n4, Value=@n5) {constMax=0.2}
   )
   B=VectorParameter() {group=Shading, parameterName=Dot Color, sortPriority=1, defaultValue=<rgba 0.3698 0 0.2705 1>}.RGB
   Alpha=If(A=Distance(A=@n9, B=@n10), B=Multiply(A=@n11, B=@n18), "A > B"=Constant(), "A < B"=Constant() {r=1}) {constB=0.2}
  )
  n20 = If(
   A=@n17
   B=ScalarParameter() {defaultValue=0.001, group=Shading, parameterName=Shadow Threshold, sliderMax=0.1}
   "A > B"=Constant() {r=1}
   "A < B"=Constant()
  )
  n21 = ParticleColor()
  MaterialFunctionCall(Emissive=Multiply(A=LinearInterpolate(A=@n1.RGB, B=@n19, Alpha=@n20), B=@n21.RGB)) {materialFunction=&MF_FX_EmissiveAdjustments}
  n22 = TextureSampleParameter2D(UVs=NamedRerouteUsage()) {
   group=Texture
   mipValueMode=TMVM_MipBias
   parameterName=Texture Alpha
   texture=/Game/Effects/Fort_Effects/Gameplay/SmokeDustExplosions/Textures/T_RibbonSmokeMasked_02
  }
  n23 = Min(A=@n22.R, B=NamedRerouteUsage())
  n24 = TextureSampleParameter2D(UVs=NamedRerouteUsage()) {
   group=Texture
   mipValueMode=TMVM_MipBias
   parameterName=Texture Erosion
   samplerType=SAMPLERTYPE_LinearColor
   texture=/Game/Effects/Fort_Effects/Gameplay/SmokeDustExplosions/Textures/T_RibbonSmokeMasked_03
  }
  n25 = ComponentMask(None=TextureCoordinate()) {r=true}
  n26 = MaterialFunctionCall!(
   Ramp=Multiply(A=@n24.R, B=OneMinus(None=@n25))
   Dissolve=DynamicParameter() {defaultValue=<rgba 0 0 0.1 1>, paramNames [Erosion, Random X, Leading Color Swipe, Param4]}.Erosion
   EdgeSharpness=ScalarParameter() {defaultValue=5, group=Shading, parameterName=Erosion Contrast, sliderMax=20}
  ) {materialFunction=/Game/Effects/Fort_Effects/Materials/_Functions/MF_Dissolve}
  n27 = ScalarParameter() {defaultValue=0.2, parameterName=Fade Start, sliderMax=1}
  n28 = ScalarParameter() {defaultValue=0.4, parameterName=Fade End, sliderMax=1}
  n29 = ComponentMask(None=TextureCoordinate()) {r=true}
  n30 = Multiply(
   A=Multiply(A=@n21.A, B=Multiply(A=@n23, B=@n26.ClampedResult))
   B=StaticSwitchParameter(True=SmoothStep(Min=@n27, Max=@n28, Value=@n29) {constMax=0.4, constMin=0.2}, False=Constant() {r=1}) {parameterName=Add Front Fade}
  )
  n31 = ScalarParameter() {defaultValue=30, group=Shading, parameterName=Depth Fade Distance, sliderMax=100}
  n32 = CameraVectorWS()
  n33 = VertexNormalWS()
  n34 = SmoothStep(
   Min=ScalarParameter() {defaultValue=0.1, parameterName=ViewFadeMin, sliderMax=1}
   Max=ScalarParameter() {defaultValue=0.2, parameterName=ViewFadeMax, sliderMax=1}
   Value=Abs(None=DotProduct(A=@n32, B=@n33))
  ) {constMax=0.1}
  n35 = Constant() {r=1}
  n36 = ScalarParameter() {defaultValue=256, parameterName=Near Camera Fade Distance, sliderMax=1024}
  Multiply(
   A=Multiply(A=DepthFade(Opacity=@n30, FadeDistance=@n31), B=StaticSwitchParameter(True=@n34, False=@n35) {parameterName=UseViewAngleFade})
   B=StaticSwitchParameter(
    True=MaterialFunctionCall(Fade Length=@n36) {materialFunction=&CameraDepthFade}.Result
    False=Constant() {r=1}
   ) {parameterName=UseNearCameraFade}
  )
  n37 = Multiply(
   A=TextureCoordinate()
   B=AppendVector(
    A=ScalarParameter() {defaultValue=1, group=Texture, parameterName=Tile X, sliderMax=5}
    B=ScalarParameter() {defaultValue=1, group=Texture, parameterName=Tile Y, sliderMax=5}
   )
  )
  n38 = AppendVector(
   A=Power(Base=ComponentMask(None=@n37) {r=true}, Exp=ScalarParameter() {defaultValue=0.5, group=Texture, parameterName=Stretch, sliderMax=2}) {constExponent=0.5}
   B=ComponentMask(None=@n37) {g=true}
  )
  n39 = Add(
   A=ComponentMask(None=@n38) {r=true}
   B=DynamicParameter() {defaultValue=<rgba 0 0 0.1 1>, paramNames [Erosion, Random X, Leading Color Swipe, Param4]}.Random X
  )
  n40 = Saturate(None=Add(A=ComponentMask(None=@n38) {g=true}, B=ScalarParameter() {group=Texture, parameterName=Offset Y, sliderMax=1, sliderMin=-1}))
  n41 = ScalarParameter() {defaultValue=0.5, group=Texture, parameterName=Speed X, sliderMax=5}
  n42 = Constant()
  NamedRerouteDeclaration(None=Panner(Coordinate=AppendVector(A=@n39, B=@n40), Speed=AppendVector(A=@n41, B=@n42))) {name=UV, nodeColor=<rgba 0 0.2266 1 1>}
  n43 = MaterialFunctionCall(UVs=TextureCoordinate(), Texture Scale=Constant2Vector() {g=1, r=1}) {materialFunction=&ScaleUVsByCenter}
  n44 = Power(Base=ComponentMask(None=@n43.UVs) {r=true}, Exp=ScalarParameter() {defaultValue=0.5, group=Shape, parameterName=Shape Warp, sliderMax=2}) {constExponent=0.25}
  n45 = ComponentMask(None=@n43.UVs) {g=true}
  NamedRerouteDeclaration(None=SphereMask(
    A=AppendVector(A=@n44, B=@n45)
    B=Constant() {r=0.5}
    Hardness=ScalarParameter() {defaultValue=1, group=Shape, parameterName=Shape Hardness, sliderMax=1}
   ) {attenuationRadius=0.5, hardnessPercent=80}) {name=Shape Mask, nodeColor=<rgba 0 1 0.3672 1>}
 ]
}
